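# Physics asset for the Military Fashion Brigade backpack cosmetic, defining bodies and joints for the money bones attached to the pack.
# Each bone has one capsule, mostly zero-length, with long, offset capsules on money_01_up and money_02_up.
# Four money bones joint directly to the pack, while money_04_up chains through money_03_up and money_up_base.
# The base-to-pack joint swings and twists freely; other joints have tighter, asymmetric swing limits.
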PhysicsAsset /BRCosmetics/Accessories/FORT_Backpacks/Backpack_F_MED_MilitaryFashion_Brigade/Meshes/F_MED_MilitaryFashion_Brigade_Pack_Physics
bodies {
 money_01_r [{shapeType=Capsule, length=0, radius=1, shapeName=money_01_r_capsule}]
 money_01_up [{
   shapeType=Capsule
   length=10
   radius=0.6
   shapeName=money_01_up_capsule
   center=<2.6902 -0.9466 0.241>
   rotation=<rot 71.2673 173.2564 -104.4396>
  }]
 money_02_r [{shapeType=Capsule, length=0, radius=1, shapeName=money_02_r_capsule}]
 money_02_up [{shapeType=Capsule, length=10, radius=0.7, shapeName=money_02_up_capsule, center=<2.4647 2.0168 0.6944>, rotation=<rot 87.0327 1.3232 0>}]
 money_03_up [{shapeType=Capsule, length=0, radius=1, shapeName=money_03_up_capsule}]
 money_04_up [{shapeType=Capsule, length=0, radius=1, shapeName=money_04_up_capsule}]
 money_up_base [{shapeType=Capsule, length=0.505, radius=0.505, shapeName=money_up_base_capsule, rotation=<rot 0 -90 0>}]
 pack [{shapeType=Capsule, length=0, radius=5, shapeName=pack_capsule}]
}
constraints [
 {bone1Name=money_04_up, bone2Name=money_03_up, swing1LimitDegrees=5, swing2LimitDegrees=40, twistLimitDegrees=15}
 {bone1Name=money_01_r, bone2Name=pack, swing1LimitDegrees=10, swing2LimitDegrees=42, twistLimitDegrees=8}
 {bone1Name=money_01_up, bone2Name=pack, swing1LimitDegrees=10, swing2LimitDegrees=40, twistLimitDegrees=8}
 {bone1Name=money_02_r, bone2Name=pack, swing1LimitDegrees=15, swing2LimitDegrees=40, twistLimitDegrees=10}
 {bone1Name=money_02_up, bone2Name=pack, swing1LimitDegrees=15, swing2LimitDegrees=40, twistLimitDegrees=12}
 {bone1Name=money_03_up, bone2Name=money_up_base, swing1LimitDegrees=7, swing2LimitDegrees=20, twistLimitDegrees=8}
 {
  bone1Name=money_up_base
  bone2Name=pack
  swing1LimitDegrees=45
  swing1Motion=Free
  swing2LimitDegrees=45
  swing2Motion=Free
  twistLimitDegrees=45
  twistMotion=Free
 }
]
